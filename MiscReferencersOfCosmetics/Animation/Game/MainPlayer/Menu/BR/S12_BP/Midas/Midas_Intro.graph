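# Level sequence for the Season 12 Battle Pass menu's Midas intro cinematic, staged in the frontend event level.
# Shot cameras cam_0100–0500 frame animated stand-in mannequins and a Black Knight mask gadget while the regular Midas and nav mannequins are hidden.
# Trophy-shelf heads, masks, the shelf and a spotlight are repositioned or hidden for particular shots, with keys restoring them near the end.
# Event tracks trigger on the frontend camera and the Midas and Maproom light switchers.
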LevelSequence /MiscReferencersOfCosmetics/Animation/Game/MainPlayer/Menu/BR/S12_BP/Midas/Midas_Intro
bindings [
 CameraComponent {
  id=#1
  parent_id=#2
  possessed_class=CineCameraComponent
  tracks [
   {
    type=MovieSceneFloatTrack
    display_name=Current Focal Length
    sections [{type=MovieSceneFloatSection, channels [None {type=MovieSceneScriptingFloatChannel, const [25,175]}]}]
   }
   {
    type=MovieSceneFloatTrack
    display_name="Manual Focus Distance (Focus Settings)"
    sections [{type=MovieSceneFloatSection, channels [None {type=MovieSceneScriptingFloatChannel, default=180, const [180,1]}]}]
   }
   {
    type=MovieSceneFloatTrack
    display_name=Current Aperture
    sections [{type=MovieSceneFloatSection, channels [None {type=MovieSceneScriptingFloatChannel, default=2.8}]}]
   }
  ]
 }
 SM_M_MED_Mastermind_Head2 {
  id=#3
  possessed_class=FortStaticMeshActor
  tracks [
   {
    type=MovieScene3DTransformTrack
    sections [{
      type=MovieScene3DTransformSection
      channels [
       Location.X {type=MovieSceneScriptingDoubleChannel, default=2939.1353, const [2939.1353,4]}
       Location.Y {type=MovieSceneScriptingDoubleChannel, default=-120499.375, const [-120499.375,4]}
       Location.Z {type=MovieSceneScriptingDoubleChannel, default=147.4466, keys [[475,147.4466],[476.0013,117.0455],[521.5863,203.4283],[732,147.4466]]}
       Rotation.X {type=MovieSceneScriptingDoubleChannel, default=0, const [0,3]}
       Rotation.Y {type=MovieSceneScriptingDoubleChannel, default=0, const [0,3]}
       Rotation.Z {type=MovieSceneScriptingDoubleChannel, default=-4.6162, const [-4.6162,3]}
       Scale.X {type=MovieSceneScriptingDoubleChannel, default=1.7, const [1.7,3]}
       Scale.Y {type=MovieSceneScriptingDoubleChannel, default=1.7, const [1.7,3]}
       Scale.Z {type=MovieSceneScriptingDoubleChannel, default=1.7, const [1.7,3]}
      ]
     }]
   }
   {
    type=MovieSceneVisibilityTrack
    sections [{
      type=MovieSceneVisibilitySection
      channels [None {type=MovieSceneScriptingBoolChannel, default=true, keys [[-73.96,true],[-6.0925,false],[521.5863,false],[541,true]]}]
     }]
   }
  ]
 }
 SM_M_MED_AUX_Robot_Head_2 {
  id=#4
  possessed_class=FortStaticMeshActor
  tracks [{
    type=MovieScene3DTransformTrack
    sections [{
      type=MovieScene3DTransformSection
      channels [
       Location.X {type=MovieSceneScriptingDoubleChannel, default=2792.613, const [2792.613,3]}
       Location.Y {type=MovieSceneScriptingDoubleChannel, default=-120497.3359, const [-120497.3359,3]}
       Location.Z {type=MovieSceneScriptingDoubleChannel, default=185.3222, keys [[540,185.3222],[541,243.0049],[732,185.3222]]}
       Rotation.X {type=MovieSceneScriptingDoubleChannel, default=0, const [0,2]}
       Rotation.Y {type=MovieSceneScriptingDoubleChannel, default=0, const [0,2]}
       Rotation.Z {type=MovieSceneScriptingDoubleChannel, default=3.4304, const [3.4304,2]}
       Scale.X {type=MovieSceneScriptingDoubleChannel, default=1.3, const [1.3,2]}
       Scale.Y {type=MovieSceneScriptingDoubleChannel, default=1.3, const [1.3,2]}
       Scale.Z {type=MovieSceneScriptingDoubleChannel, default=1.3, const [1.3,2]}
      ]
     }]
   }]
 }
 StaticMeshComponent0 {
  id=#5
  parent_id=#6
  possessed_class=StaticMeshComponent
  tracks [{
    type=MovieSceneComponentMaterialTrack
    display_name=Material Element 0
    sections [{
      type=MovieSceneParameterSection
      end_frame=731
      start_frame=-8
      channels [CINE_MidasControl {type=MovieSceneScriptingFloatChannel, default=1, const [1,1]}]
     }]
   }]
 }
 SM_M_MED_Crazy_Eight_Head2 {
  id=#7
  possessed_class=FortStaticMeshActor
  tracks [{
    type=MovieScene3DTransformTrack
    sections [{
      type=MovieScene3DTransformSection
      channels [
       Location.X {type=MovieSceneScriptingDoubleChannel, default=2937.4844, const [2937.4844,4]}
       Location.Y {type=MovieSceneScriptingDoubleChannel, default=-120488.5469, thinned_from=4, keys [[475,-120488.5469],[732,-120488.5469]]}
       Location.Z {type=MovieSceneScriptingDoubleChannel, default=186.3152, keys [[475,186.3152],[476.0013,160.3831],[541,242.565],[732,186.3152]]}
       Rotation.X {type=MovieSceneScriptingDoubleChannel, default=0, const [0,2]}
       Rotation.Y {type=MovieSceneScriptingDoubleChannel, default=0, const [0,2]}
       Rotation.Z {type=MovieSceneScriptingDoubleChannel, default=0.2866, const [0.2866,2]}
       Scale.X {type=MovieSceneScriptingDoubleChannel, default=1.3, const [1.3,2]}
       Scale.Y {type=MovieSceneScriptingDoubleChannel, default=1.3, const [1.3,2]}
       Scale.Z {type=MovieSceneScriptingDoubleChannel, default=1.3, const [1.3,2]}
      ]
     }]
   }]
 }
 cam_0400 {
  id=#8
  possessed_class=CineCameraActor
  tracks [
   {
    type=MovieScene3DTransformTrack
    sections [{
      type=MovieScene3DTransformSection
      channels [
       Location.X {type=MovieSceneScriptingDoubleChannel, default=3113.6948, const [3113.6948,66]}
       Location.Y {type=MovieSceneScriptingDoubleChannel, default=-120482.9141, const [-120482.9141,66]}
       Location.Z {type=MovieSceneScriptingDoubleChannel, default=177.7992, const [177.7992,66]}
       Rotation.X {type=MovieSceneScriptingDoubleChannel, default=0.7628, const [360.7628,66]}
       Rotation.Y {type=MovieSceneScriptingDoubleChannel, default=3.1431, const [-356.8569,66]}
       Rotation.Z {type=MovieSceneScriptingDoubleChannel, default=-179.388, const [-179.388,66]}
       Scale.X {type=MovieSceneScriptingDoubleChannel, default=1, const [1,66]}
       Scale.Y {type=MovieSceneScriptingDoubleChannel, default=1, const [1,66]}
       Scale.Z {type=MovieSceneScriptingDoubleChannel, default=1, const [1,66]}
      ]
     }]
   }
   {type=MovieSceneSpawnTrack, sections [{type=MovieSceneSpawnSection, channels [None {type=MovieSceneScriptingBoolChannel, default=true}]}]}
  ]
 }
 BP12_MIdas_Mannequin {
  id=#9
  possessed_class=BP12_MIdas_Mannequin_C
  tracks [
   {
    type=MovieScene3DTransformTrack
    sections [{
      type=MovieScene3DTransformSection
      channels [
       Location.X {type=MovieSceneScriptingDoubleChannel, default=0}
       Location.Y {type=MovieSceneScriptingDoubleChannel, default=0}
       Location.Z {type=MovieSceneScriptingDoubleChannel, default=0}
       Rotation.X {type=MovieSceneScriptingDoubleChannel, default=0}
       Rotation.Y {type=MovieSceneScriptingDoubleChannel, default=0}
       Rotation.Z {type=MovieSceneScriptingDoubleChannel, default=0}
       Scale.X {type=MovieSceneScriptingDoubleChannel, default=1}
       Scale.Y {type=MovieSceneScriptingDoubleChannel, default=1}
       Scale.Z {type=MovieSceneScriptingDoubleChannel, default=1}
      ]
     }]
   }
   {type=MovieSceneSkeletalAnimationTrack, sections []}
   {
    type=MovieSceneVisibilityTrack
    sections [{
      type=MovieSceneVisibilitySection
      channels [None {type=MovieSceneScriptingBoolChannel, default=true, keys [[-0.6375,true],[0.0075,false],[730.9813,true]]}]
     }]
   }
  ]
 }
 BP_Frontend_EventLevel_NavObject_Mannequin3 {
  id=#10
  possessed_class=SoldierMannequin_Male_C
  tracks [
   {
    type=MovieScene3DTransformTrack
    sections [{
      type=MovieScene3DTransformSection
      channels [
       Location.X {type=MovieSceneScriptingDoubleChannel, default=2290}
       Location.Y {type=MovieSceneScriptingDoubleChannel, default=-120000}
       Location.Z {type=MovieSceneScriptingDoubleChannel, default=8}
       Rotation.X {type=MovieSceneScriptingDoubleChannel, default=0}
       Rotation.Y {type=MovieSceneScriptingDoubleChannel, default=0}
       Rotation.Z {type=MovieSceneScriptingDoubleChannel, default=0}
       Scale.X {type=MovieSceneScriptingDoubleChannel, default=1}
       Scale.Y {type=MovieSceneScriptingDoubleChannel, default=1}
       Scale.Z {type=MovieSceneScriptingDoubleChannel, default=1}
      ]
     }]
   }
   {
    type=MovieSceneSkeletalAnimationTrack
    sections [
     {type=MovieSceneSkeletalAnimationSection, end_frame=125, start_frame=0}
     {type=MovieSceneSkeletalAnimationSection, end_frame=242, start_frame=125}
     {type=MovieSceneSkeletalAnimationSection, end_frame=541, start_frame=476}
    ]
   }
   {
    type=MovieSceneVisibilityTrack
    sections [{
      type=MovieSceneVisibilitySection
      channels [None {type=MovieSceneScriptingBoolChannel, default=false, keys [[-7.1175,false],[0.1413,true],[541,false]]}]
     }]
   }
  ]
 }
 SM_M_MED_PizzaDiety_FaceAcc2 {
  id=#11
  possessed_class=FortStaticMeshActor
  tracks [{
    type=MovieScene3DTransformTrack
    sections [{
      type=MovieScene3DTransformSection
      channels [
       Location.X {type=MovieSceneScriptingDoubleChannel, default=2833.2275, const [2833.2275,3]}
       Location.Y {type=MovieSceneScriptingDoubleChannel, default=-120490.8281, const [-120490.8281,3]}
       Location.Z {type=MovieSceneScriptingDoubleChannel, default=97.7346, keys [[540,97.7346],[541,158.1623],[732,97.7346]]}
       Rotation.X {type=MovieSceneScriptingDoubleChannel, default=0, const [0,2]}
       Rotation.Y {type=MovieSceneScriptingDoubleChannel, default=0, const [0,2]}
       Rotation.Z {type=MovieSceneScriptingDoubleChannel, default=-7.8374, const [-7.8374,2]}
       Scale.X {type=MovieSceneScriptingDoubleChannel, default=1.2, const [1.2,2]}
       Scale.Y {type=MovieSceneScriptingDoubleChannel, default=1.2, const [1.2,2]}
       Scale.Z {type=MovieSceneScriptingDoubleChannel, default=1.2, const [1.2,2]}
      ]
     }]
   }]
 }
 cam_0300 {
  id=#12
  possessed_class=CineCameraActor
  tracks [
   {
    type=MovieScene3DTransformTrack
    sections [{
      type=MovieScene3DTransformSection
      channels [
       Location.X {type=MovieSceneScriptingDoubleChannel, default=2940.1858, thinned_from=235, keys [[242.0013,2940.1858],[476.0013,2939.5239]]}
       Location.Y {type=MovieSceneScriptingDoubleChannel, default=-120253.8984, thinned_from=235, keys [[242.0013,-120253.8984],[476.0013,-120260.6953]]}
       Location.Z {type=MovieSceneScriptingDoubleChannel, default=223.1779, thinned_from=235, keys [[242.0013,223.1779],[476.0013,218.4836]]}
       Rotation.X {type=MovieSceneScriptingDoubleChannel, default=-5.2447, thinned_from=235, keys [[242.0013,354.7553],[476.0013,354.6911]]}
       Rotation.Y {type=MovieSceneScriptingDoubleChannel, default=-39.1767, thinned_from=235, keys [[242.0013,-399.1767],[476.0013,-399.2422]]}
       Rotation.Z {type=MovieSceneScriptingDoubleChannel, default=-105.336, thinned_from=235, keys [[242.0013,-105.336],[476.0013,-105.2342]]}
       Scale.X {type=MovieSceneScriptingDoubleChannel, default=1, const [1,235]}
       Scale.Y {type=MovieSceneScriptingDoubleChannel, default=1, const [1,235]}
       Scale.Z {type=MovieSceneScriptingDoubleChannel, default=1, const [1,235]}
      ]
     }]
   }
   {type=MovieSceneSpawnTrack, sections [{type=MovieSceneSpawnSection, channels [None {type=MovieSceneScriptingBoolChannel, default=true}]}]}
  ]
 }
 SM_F_MED_Space_Suit_FaceAcc2 {
  id=#13
  possessed_class=FortStaticMeshActor
  tracks [{
    type=MovieScene3DTransformTrack
    sections [{
      type=MovieScene3DTransformSection
      channels [
       Location.X {type=MovieSceneScriptingDoubleChannel, default=2882.6387, keys [[540,2882.6387],[541,2890.1404],[732,2882.6387]]}
       Location.Y {type=MovieSceneScriptingDoubleChannel, default=-120498.1406, const [-120498.1406,3]}
       Location.Z {type=MovieSceneScriptingDoubleChannel, default=241.544, keys [[540,241.544],[541,250.0242],[732,241.544]]}
       Rotation.X {type=MovieSceneScriptingDoubleChannel, default=0, const [0,3]}
       Rotation.Y {type=MovieSceneScriptingDoubleChannel, default=0, const [0,3]}
       Rotation.Z {type=MovieSceneScriptingDoubleChannel, default=-5.8087, const [-5.8087,3]}
       Scale.X {type=MovieSceneScriptingDoubleChannel, default=1.3, const [1.3,3]}
       Scale.Y {type=MovieSceneScriptingDoubleChannel, default=1.3, const [1.3,3]}
       Scale.Z {type=MovieSceneScriptingDoubleChannel, default=1.3, const [1.3,3]}
      ]
     }]
   }]
 }
 cam_0500 {
  id=#2
  possessed_class=CineCameraActor
  tracks [
   {
    type=MovieScene3DTransformTrack
    sections [{
      type=MovieScene3DTransformSection
      channels [
       Location.X {type=MovieSceneScriptingDoubleChannel, default=2894.0054, thinned_from=175, keys [[541,2894.0054],[715,2905.0879]]}
       Location.Y {type=MovieSceneScriptingDoubleChannel, default=-120345.8984, thinned_from=175, keys [[541,-120345.8984],[715,-119970.5625]]}
       Location.Z {type=MovieSceneScriptingDoubleChannel, default=200.2986, thinned_from=175, keys [[541,200.2986],[704,149.5083],[715,148.6319]]}
       Rotation.X {type=MovieSceneScriptingDoubleChannel, default=-0.1213, thinned_from=175, keys [[541,359.8788],[715,360.0466]]}
       Rotation.Y {type=MovieSceneScriptingDoubleChannel, default=1.3545, thinned_from=175, keys [[541,-358.6455],[715,-346.3645]]}
       Rotation.Z {type=MovieSceneScriptingDoubleChannel, default=-96.3495, thinned_from=175, keys [[541,-96.3495],[715,-92.22]]}
       Scale.X {type=MovieSceneScriptingDoubleChannel, default=1, const [1,175]}
       Scale.Y {type=MovieSceneScriptingDoubleChannel, default=1, const [1,175]}
       Scale.Z {type=MovieSceneScriptingDoubleChannel, default=1, const [1,175]}
      ]
     }]
   }
   {type=MovieSceneSpawnTrack, sections [{type=MovieSceneSpawnSection, channels [None {type=MovieSceneScriptingBoolChannel, default=true}]}]}
  ]
 }
 CameraComponent {
  id=#14
  parent_id=#12
  possessed_class=CineCameraComponent
  tracks [
   {
    type=MovieSceneFloatTrack
    display_name=Current Focal Length
    sections [{type=MovieSceneFloatSection, channels [None {type=MovieSceneScriptingFloatChannel, const [35,235]}]}]
   }
   {
    type=MovieSceneFloatTrack
    display_name="Manual Focus Distance (Focus Settings)"
    sections [{type=MovieSceneFloatSection, channels [None {type=MovieSceneScriptingFloatChannel, default=73.8069, const [73.8069,1]}]}]
   }
   {
    type=MovieSceneFloatTrack
    display_name=Current Aperture
    sections [{type=MovieSceneFloatSection, channels [None {type=MovieSceneScriptingFloatChannel, default=2.8}]}]
   }
  ]
 }
 LightComponent0 {
  id=#15
  parent_id=#16
  possessed_class=SpotLightComponent
  tracks [
   {
    type=MovieSceneFloatTrack
    display_name=Intensity
    sections [{type=MovieSceneFloatSection, channels [None {type=MovieSceneScriptingFloatChannel, default=30}]}]
   }
   {
    type=MovieSceneColorTrack
    display_name=Light Color
    sections [{
      type=MovieSceneColorSection
      channels [
       Color.R {type=MovieSceneScriptingFloatChannel, default=1}
       Color.G {type=MovieSceneScriptingFloatChannel, default=0.7011}
       Color.B {type=MovieSceneScriptingFloatChannel, default=0.5776}
       Color.A {type=MovieSceneScriptingFloatChannel, default=1}
      ]
     }]
   }
  ]
 }
 SM_M_MED_CuChulainn_FaceAcc2 {
  id=#17
  possessed_class=FortStaticMeshActor
  tracks [{
    type=MovieScene3DTransformTrack
    sections [{
      type=MovieScene3DTransformSection
      channels [
       Location.X {type=MovieSceneScriptingDoubleChannel, default=2934.4055, const [2934.4055,6]}
       Location.Y {type=MovieSceneScriptingDoubleChannel, default=-120498.1875, const [-120498.1875,6]}
       Location.Z {
        type=MovieSceneScriptingDoubleChannel
        default=249.8979
        thinned_from=6
        keys [[475,249.8979],[540,249.8979],[541,307.7414],[731,306],[732,249.8979]]
       }
       Rotation.X {type=MovieSceneScriptingDoubleChannel, default=0, const [0,2]}
       Rotation.Y {type=MovieSceneScriptingDoubleChannel, default=0, const [0,2]}
       Rotation.Z {type=MovieSceneScriptingDoubleChannel, default=3.3936, const [3.3936,2]}
       Scale.X {type=MovieSceneScriptingDoubleChannel, default=1.3, const [1.3,2]}
       Scale.Y {type=MovieSceneScriptingDoubleChannel, default=1.3, const [1.3,2]}
       Scale.Z {type=MovieSceneScriptingDoubleChannel, default=1.3, const [1.3,2]}
      ]
     }]
   }]
 }
 SM_M_MED_WWII_Pilot_Sci_Fi_Head2 {
  id=#18
  possessed_class=FortStaticMeshActor
  tracks [{
    type=MovieScene3DTransformTrack
    sections [{
      type=MovieScene3DTransformSection
      channels [
       Location.X {type=MovieSceneScriptingDoubleChannel, default=2888.4412, const [2888.4412,4]}
       Location.Y {type=MovieSceneScriptingDoubleChannel, default=-120499.4453, const [-120499.4453,4]}
       Location.Z {type=MovieSceneScriptingDoubleChannel, default=190.5944, keys [[475,190.5944],[476.0013,208.8528],[541,294.756],[732,190.5944]]}
       Rotation.X {type=MovieSceneScriptingDoubleChannel, default=0, const [0,3]}
       Rotation.Y {type=MovieSceneScriptingDoubleChannel, default=0, const [0,3]}
       Rotation.Z {type=MovieSceneScriptingDoubleChannel, default=-2.9474, const [-2.9474,3]}
       Scale.X {type=MovieSceneScriptingDoubleChannel, default=1.3, const [1.3,3]}
       Scale.Y {type=MovieSceneScriptingDoubleChannel, default=1.3, const [1.3,3]}
       Scale.Z {type=MovieSceneScriptingDoubleChannel, default=1.3, const [1.3,3]}
      ]
     }]
   }]
 }
 BP_Frontend_EventLevel_NavObject_MapTable {
  id=#19
  possessed_class=BP_Frontend_EventLevel_NavObject_MapTable_C
  tracks [{
    type=MovieScene3DTransformTrack
    sections [{
      type=MovieScene3DTransformSection
      channels [
       Location.X {type=MovieSceneScriptingDoubleChannel, default=3044.8689, thinned_from=3, keys [[-8.105,3044.8689],[-7.11,3124.8777]]}
       Location.Y {type=MovieSceneScriptingDoubleChannel, default=-119836, thinned_from=3, keys [[-8.105,-119836],[-7.11,-119684.8594]]}
       Location.Z {type=MovieSceneScriptingDoubleChannel, default=8, const [8,3]}
       Rotation.X {type=MovieSceneScriptingDoubleChannel, default=0, const [0,1]}
       Rotation.Y {type=MovieSceneScriptingDoubleChannel, default=0, const [0,1]}
       Rotation.Z {type=MovieSceneScriptingDoubleChannel, default=0, const [0,1]}
       Scale.X {type=MovieSceneScriptingDoubleChannel, default=1, const [1,1]}
       Scale.Y {type=MovieSceneScriptingDoubleChannel, default=1, const [1,1]}
       Scale.Z {type=MovieSceneScriptingDoubleChannel, default=1, const [1,1]}
      ]
     }]
   }]
 }
 CameraComponent {
  id=#20
  parent_id=#21
  possessed_class=CineCameraComponent
  tracks [
   {
    type=MovieSceneFloatTrack
    display_name=Current Focal Length
    sections [{type=MovieSceneFloatSection, channels [None {type=MovieSceneScriptingFloatChannel, default=35}]}]
   }
   {
    type=MovieSceneFloatTrack
    display_name="Manual Focus Distance (Focus Settings)"
    sections [{type=MovieSceneFloatSection, channels [None {type=MovieSceneScriptingFloatChannel, default=100000}]}]
   }
   {
    type=MovieSceneFloatTrack
    display_name=Current Aperture
    sections [{type=MovieSceneFloatSection, channels [None {type=MovieSceneScriptingFloatChannel, default=2.8}]}]
   }
  ]
 }
 BP_Frontend_EventLevel_NavObject_Mannequin {
  id=#22
  possessed_class=BP_Frontend_EventLevel_NavObject_Mannequin_C
  tracks [{
    type=MovieSceneVisibilityTrack
    sections [{
      type=MovieSceneVisibilitySection
      channels [None {type=MovieSceneScriptingBoolChannel, default=true, keys [[-0.005,true],[0,false],[731,true]]}]
     }]
   }]
 }
 SM_M_MED_BlackKnight_01_FaceAcc2 {
  id=#6
  possessed_class=FortStaticMeshActor
  tracks [
   {
    type=MovieScene3DTransformTrack
    sections [{
      type=MovieScene3DTransformSection
      channels [
       Location.X {type=MovieSceneScriptingDoubleChannel, default=2892.0208, const [2892.0208,4]}
       Location.Y {type=MovieSceneScriptingDoubleChannel, default=-120495.4141, const [-120495.4141,4]}
       Location.Z {type=MovieSceneScriptingDoubleChannel, default=157.4585, keys [[475,157.4585],[476.0013,124.7917],[528.6463,210.2905],[732,157]]}
       Rotation.X {type=MovieSceneScriptingDoubleChannel, default=0, const [0,3]}
       Rotation.Y {type=MovieSceneScriptingDoubleChannel, default=0, const [0,3]}
       Rotation.Z {type=MovieSceneScriptingDoubleChannel, default=-8.1979, const [-8.1979,3]}
       Scale.X {type=MovieSceneScriptingDoubleChannel, default=1.3, const [1.3,3]}
       Scale.Y {type=MovieSceneScriptingDoubleChannel, default=1.3, const [1.3,3]}
       Scale.Z {type=MovieSceneScriptingDoubleChannel, default=1.3, const [1.3,3]}
      ]
     }]
   }
   {
    type=MovieSceneVisibilityTrack
    sections [{
      type=MovieSceneVisibilitySection
      channels [None {type=MovieSceneScriptingBoolChannel, default=true, keys [[-107.4013,true],[-6.0925,false],[528.6463,false],[541,true]]}]
     }]
   }
  ]
 }
 SM_M_MED_DurrburgerHero_FaceAcc2 {
  id=#23
  possessed_class=FortStaticMeshActor
  tracks [{
    type=MovieScene3DTransformTrack
    sections [{
      type=MovieScene3DTransformSection
      channels [
       Location.X {type=MovieSceneScriptingDoubleChannel, default=2798.5779, const [2798.5779,3]}
       Location.Y {type=MovieSceneScriptingDoubleChannel, default=-120490.7109, const [-120490.7109,3]}
       Location.Z {type=MovieSceneScriptingDoubleChannel, default=141.7205, keys [[540,141.7205],[541,199.0712],[732,141.7205]]}
       Rotation.X {type=MovieSceneScriptingDoubleChannel, default=0, const [0,2]}
       Rotation.Y {type=MovieSceneScriptingDoubleChannel, default=0, const [0,2]}
       Rotation.Z {type=MovieSceneScriptingDoubleChannel, default=12.6781, const [12.6781,2]}
       Scale.X {type=MovieSceneScriptingDoubleChannel, default=1.2, const [1.2,2]}
       Scale.Y {type=MovieSceneScriptingDoubleChannel, default=1.2, const [1.2,2]}
       Scale.Z {type=MovieSceneScriptingDoubleChannel, default=1.2, const [1.2,2]}
      ]
     }]
   }]
 }
 Camera_Root_Movement {
  id=#24
  possessed_class=Frontend_EventLevel_Camera_C
  tracks [
   {
    type=MovieScene3DTransformTrack
    sections [{
      type=MovieScene3DTransformSection
      channels [
       Location.X {type=MovieSceneScriptingDoubleChannel, default=2803.4331, const [2803.4331,1]}
       Location.Y {type=MovieSceneScriptingDoubleChannel, default=-120226.7109, const [-120226.7109,1]}
       Location.Z {type=MovieSceneScriptingDoubleChannel, default=107.8428, const [107.8428,1]}
       Rotation.X {type=MovieSceneScriptingDoubleChannel, default=0, const [0,1]}
       Rotation.Y {type=MovieSceneScriptingDoubleChannel, default=-2.4004, const [-2.4004,1]}
       Rotation.Z {type=MovieSceneScriptingDoubleChannel, default=-214.4646, const [-214.4646,1]}
       Scale.X {type=MovieSceneScriptingDoubleChannel, default=1, const [1,1]}
       Scale.Y {type=MovieSceneScriptingDoubleChannel, default=1, const [1,1]}
       Scale.Z {type=MovieSceneScriptingDoubleChannel, default=1, const [1,1]}
      ]
     }]
   }
   {
    type=MovieSceneEventTrack
    sections [{
      type=MovieSceneEventTriggerSection
      channels [None {type=MovieSceneScriptingEventChannel, const ["<Struct 'MovieSceneEvent' (0x8518df53660) {}>", 1]}]
     }]
   }
  ]
 }
 CameraComponent {
  id=#25
  parent_id=#8
  possessed_class=CineCameraComponent
  tracks [
   {
    type=MovieSceneFloatTrack
    display_name=Current Focal Length
    sections [{type=MovieSceneFloatSection, channels [None {type=MovieSceneScriptingFloatChannel, const [85,66]}]}]
   }
   {
    type=MovieSceneFloatTrack
    display_name="Manual Focus Distance (Focus Settings)"
    sections [{type=MovieSceneFloatSection, channels [None {type=MovieSceneScriptingFloatChannel, default=235.2252}]}]
   }
   {
    type=MovieSceneFloatTrack
    display_name=Current Aperture
    sections [{type=MovieSceneFloatSection, channels [None {type=MovieSceneScriptingFloatChannel, default=2.8}]}]
   }
  ]
 }
 CameraComponent {
  id=#26
  parent_id=#24
  possessed_class=CameraComponent
  tracks [{
    type=MovieSceneFloatTrack
    display_name=Field Of View
    sections [{type=MovieSceneFloatSection, channels [None {type=MovieSceneScriptingFloatChannel, default=80, const [80,2]}]}]
   }]
 }
 SM_F_MED_Neon_Cat_T3_Head_2 {
  id=#27
  possessed_class=FortStaticMeshActor
  tracks [{
    type=MovieScene3DTransformTrack
    sections [{
      type=MovieScene3DTransformSection
      channels [
       Location.X {type=MovieSceneScriptingDoubleChannel, default=2795.2583, const [2795.2583,3]}
       Location.Y {type=MovieSceneScriptingDoubleChannel, default=-120491.7031, const [-120491.7031,3]}
       Location.Z {type=MovieSceneScriptingDoubleChannel, default=240.4786, keys [[540,240.4786],[541,302.6968],[732,240.4786]]}
       Rotation.X {type=MovieSceneScriptingDoubleChannel, default=0, const [0,3]}
       Rotation.Y {type=MovieSceneScriptingDoubleChannel, default=0, const [0,3]}
       Rotation.Z {type=MovieSceneScriptingDoubleChannel, default=-10.5657, const [-10.5657,3]}
       Scale.X {type=MovieSceneScriptingDoubleChannel, default=1.3, const [1.3,3]}
       Scale.Y {type=MovieSceneScriptingDoubleChannel, default=1.3, const [1.3,3]}
       Scale.Z {type=MovieSceneScriptingDoubleChannel, default=1.3, const [1.3,3]}
      ]
     }]
   }]
 }
 CameraComponent {
  id=#28
  parent_id=#29
  possessed_class=CineCameraComponent
  tracks [
   {
    type=MovieSceneFloatTrack
    display_name=Current Focal Length
    sections [{type=MovieSceneFloatSection, channels [None {type=MovieSceneScriptingFloatChannel, const [15,126]}]}]
   }
   {
    type=MovieSceneFloatTrack
    display_name="Manual Focus Distance (Focus Settings)"
    sections [{type=MovieSceneFloatSection, channels [None {type=MovieSceneScriptingFloatChannel, default=183.2412}]}]
   }
   {
    type=MovieSceneFloatTrack
    display_name=Current Aperture
    sections [{type=MovieSceneFloatSection, channels [None {type=MovieSceneScriptingFloatChannel, default=2.8}]}]
   }
  ]
 }
 SM_M_LRG_Big_ChuggusHead2 {
  id=#30
  possessed_class=FortStaticMeshActor
  tracks [{
    type=MovieScene3DTransformTrack
    sections [{
      type=MovieScene3DTransformSection
      channels [
       Location.X {type=MovieSceneScriptingDoubleChannel, default=2836.3535, const [2836.3535,3]}
       Location.Y {type=MovieSceneScriptingDoubleChannel, default=-120499.4453, const [-120499.4453,3]}
       Location.Z {type=MovieSceneScriptingDoubleChannel, default=234.0523, keys [[540,234.0523],[541,243.1904],[732,234.0523]]}
       Rotation.X {type=MovieSceneScriptingDoubleChannel, default=0, const [0,2]}
       Rotation.Y {type=MovieSceneScriptingDoubleChannel, default=0, const [0,2]}
       Rotation.Z {type=MovieSceneScriptingDoubleChannel, default=2.3003, const [2.3003,2]}
       Scale.X {type=MovieSceneScriptingDoubleChannel, default=1.3, const [1.3,2]}
       Scale.Y {type=MovieSceneScriptingDoubleChannel, default=1.3, const [1.3,2]}
       Scale.Z {type=MovieSceneScriptingDoubleChannel, default=1.3, const [1.3,2]}
      ]
     }]
   }]
 }
 _Midas_Black_Knight_Mask_01_Gadget {
  id=#31
  possessed_class=SkeletalMeshActor
  tracks [
   {
    type=MovieScene3DTransformTrack
    sections [{
      type=MovieScene3DTransformSection
      channels [
       Location.X {type=MovieSceneScriptingDoubleChannel, default=2290}
       Location.Y {type=MovieSceneScriptingDoubleChannel, default=-120000}
       Location.Z {type=MovieSceneScriptingDoubleChannel, default=8}
       Rotation.X {type=MovieSceneScriptingDoubleChannel, default=0}
       Rotation.Y {type=MovieSceneScriptingDoubleChannel, default=0}
       Rotation.Z {type=MovieSceneScriptingDoubleChannel, default=0}
       Scale.X {type=MovieSceneScriptingDoubleChannel, default=1}
       Scale.Y {type=MovieSceneScriptingDoubleChannel, default=1}
       Scale.Z {type=MovieSceneScriptingDoubleChannel, default=1}
      ]
     }]
   }
   {
    type=MovieSceneSkeletalAnimationTrack
    sections [
     {type=MovieSceneSkeletalAnimationSection, end_frame=125, start_frame=0}
     {type=MovieSceneSkeletalAnimationSection, end_frame=242, start_frame=125}
     {type=MovieSceneSkeletalAnimationSection, end_frame=476, start_frame=242}
     {type=MovieSceneSkeletalAnimationSection, end_frame=541, start_frame=476}
    ]
   }
   {
    type=MovieSceneVisibilityTrack
    sections [{
      type=MovieSceneVisibilitySection
      channels [None {type=MovieSceneScriptingBoolChannel, default=false, keys [[-8.0038,false],[3.3175,true],[731.9987,false]]}]
     }]
   }
  ]
 }
 SM_Agency_Midas_TrophyShelf_3 {
  id=#32
  possessed_class=FortStaticMeshActor
  tracks [{
    type=MovieScene3DTransformTrack
    sections [{
      type=MovieScene3DTransformSection
      channels [
       Location.X {type=MovieSceneScriptingDoubleChannel, default=2865.9934, const [2865.9934,6]}
       Location.Y {type=MovieSceneScriptingDoubleChannel, default=-120511.2031, thinned_from=5, keys [[475,-120511.2031],[732,-120511.2031]]}
       Location.Z {
        type=MovieSceneScriptingDoubleChannel
        default=53.2657
        thinned_from=6
        keys [[475,8.7189],[476.0013,53.2657],[541,73],[731,73],[732,53.2657]]
       }
       Rotation.X {type=MovieSceneScriptingDoubleChannel, default=0, const [0,4]}
       Rotation.Y {type=MovieSceneScriptingDoubleChannel, default=0, const [0,4]}
       Rotation.Z {type=MovieSceneScriptingDoubleChannel, default=-2.3245, const [-2.3245,4]}
       Scale.X {type=MovieSceneScriptingDoubleChannel, default=0.8, keys [[475,0.8],[476.0013,0.8],[541,0.9],[731,0.9],[732,0.8]]}
       Scale.Y {type=MovieSceneScriptingDoubleChannel, default=0.8, keys [[475,0.8],[476.0013,0.905],[541,1],[731,1],[732,0.8]]}
       Scale.Z {type=MovieSceneScriptingDoubleChannel, default=0.8, const [0.8,5]}
      ]
     }]
   }]
 }
 cam_0200 {
  id=#33
  possessed_class=CineCameraActor
  tracks [
   {
    type=MovieScene3DTransformTrack
    sections [{
      type=MovieScene3DTransformSection
      channels [
       Location.X {type=MovieSceneScriptingDoubleChannel, default=3024.7202, const [3024.7202,118]}
       Location.Y {type=MovieSceneScriptingDoubleChannel, default=-120414.0234, const [-120414.0234,118]}
       Location.Z {type=MovieSceneScriptingDoubleChannel, default=90.5338, const [90.5338,118]}
       Rotation.X {type=MovieSceneScriptingDoubleChannel, default=-0.3973, thinned_from=118, keys [[125.0013,359.6027],[242.0013,359.5239]]}
       Rotation.Y {type=MovieSceneScriptingDoubleChannel, default=27.8703, thinned_from=118, keys [[125.0013,-332.1297],[242.0013,-332.1677]]}
       Rotation.Z {type=MovieSceneScriptingDoubleChannel, default=138.3854, thinned_from=118, keys [[125.0013,138.3855],[242.0013,138.2163]]}
       Scale.X {type=MovieSceneScriptingDoubleChannel, default=1, const [1,118]}
       Scale.Y {type=MovieSceneScriptingDoubleChannel, default=1, const [1,118]}
       Scale.Z {type=MovieSceneScriptingDoubleChannel, default=1, const [1,118]}
      ]
     }]
   }
   {type=MovieSceneSpawnTrack, sections [{type=MovieSceneSpawnSection, channels [None {type=MovieSceneScriptingBoolChannel, default=true}]}]}
  ]
 }
 SM_M_MED_Commando_14_FaceAcc2 {
  id=#34
  possessed_class=FortStaticMeshActor
  tracks [{
    type=MovieScene3DTransformTrack
    sections [{
      type=MovieScene3DTransformSection
      channels [
       Location.X {type=MovieSceneScriptingDoubleChannel, default=2842.9463, const [2842.9463,3]}
       Location.Y {type=MovieSceneScriptingDoubleChannel, default=-120493.5234, const [-120493.5234,3]}
       Location.Z {type=MovieSceneScriptingDoubleChannel, default=186.381, keys [[540,186.381],[541,199.7408],[732,186.381]]}
       Rotation.X {type=MovieSceneScriptingDoubleChannel, default=0, const [0,2]}
       Rotation.Y {type=MovieSceneScriptingDoubleChannel, default=0, const [0,2]}
       Rotation.Z {type=MovieSceneScriptingDoubleChannel, default=3.4704, const [3.4704,2]}
       Scale.X {type=MovieSceneScriptingDoubleChannel, default=1.3, const [1.3,2]}
       Scale.Y {type=MovieSceneScriptingDoubleChannel, default=1.3, const [1.3,2]}
       Scale.Z {type=MovieSceneScriptingDoubleChannel, default=1.3, const [1.3,2]}
      ]
     }]
   }]
 }
 BP_Frontend_EventLevel_NavObject_Mannequin4 {
  id=#35
  possessed_class=SoldierMannequin_Male_C
  tracks [
   {
    type=MovieScene3DTransformTrack
    sections [{
      type=MovieScene3DTransformSection
      channels [
       Location.X {type=MovieSceneScriptingDoubleChannel, default=2290}
       Location.Y {type=MovieSceneScriptingDoubleChannel, default=-120000}
       Location.Z {type=MovieSceneScriptingDoubleChannel, default=8}
       Rotation.X {type=MovieSceneScriptingDoubleChannel, default=0}
       Rotation.Y {type=MovieSceneScriptingDoubleChannel, default=0}
       Rotation.Z {type=MovieSceneScriptingDoubleChannel, default=0}
       Scale.X {type=MovieSceneScriptingDoubleChannel, default=1}
       Scale.Y {type=MovieSceneScriptingDoubleChannel, default=1}
       Scale.Z {type=MovieSceneScriptingDoubleChannel, default=1}
      ]
     }]
   }
   {
    type=MovieSceneSkeletalAnimationTrack
    sections [
     {type=MovieSceneSkeletalAnimationSection, end_frame=242, start_frame=191}
     {type=MovieSceneSkeletalAnimationSection, end_frame=476, start_frame=242}
     {type=MovieSceneSkeletalAnimationSection, end_frame=541, start_frame=521}
     {type=MovieSceneSkeletalAnimationSection, end_frame=715, start_frame=541}
    ]
   }
   {
    type=MovieSceneVisibilityTrack
    sections [{
      type=MovieSceneVisibilitySection
      channels [None {type=MovieSceneScriptingBoolChannel, default=false, keys [[160.5025,false],[242.0013,true],[476.0013,false],[541,true],[732,false]]}]
     }]
   }
  ]
 }
 CameraComponent {
  id=#36
  parent_id=#33
  possessed_class=CineCameraComponent
  tracks [
   {
    type=MovieSceneFloatTrack
    display_name=Current Focal Length
    sections [{type=MovieSceneFloatSection, channels [None {type=MovieSceneScriptingFloatChannel, const [85,118]}]}]
   }
   {
    type=MovieSceneFloatTrack
    display_name="Manual Focus Distance (Focus Settings)"
    sections [{
      type=MovieSceneFloatSection
      channels [None {type=MovieSceneScriptingFloatChannel, default=215.9299, keys [[125.0013,240.4609],[175.0238,215.9299]]}]
     }]
   }
   {
    type=MovieSceneFloatTrack
    display_name=Current Aperture
    sections [{type=MovieSceneFloatSection, channels [None {type=MovieSceneScriptingFloatChannel, default=2.8}]}]
   }
  ]
 }
 BP_Battlepass_LightSwitcher_Midas {
  id=#37
  possessed_class=BP_Battlepass_LightSwitcher_C
  tracks [{
    type=MovieSceneEventTrack
    sections [{
      type=MovieSceneEventTriggerSection
      channels [None {type=MovieSceneScriptingEventChannel, const ["<Struct 'MovieSceneEvent' (0x8518df54920) {}>", 1]}]
     }]
   }]
 }
 SM_M_MED_Rhino_Head2 {
  id=#38
  possessed_class=FortStaticMeshActor
  tracks [
   {
    type=MovieScene3DTransformTrack
    sections [{
      type=MovieScene3DTransformSection
      channels [
       Location.X {type=MovieSceneScriptingDoubleChannel, default=2907.7817, const [2907.7817,1]}
       Location.Y {type=MovieSceneScriptingDoubleChannel, default=-120501.5469, const [-120501.5469,1]}
       Location.Z {type=MovieSceneScriptingDoubleChannel, default=88.9652, const [145,3]}
       Rotation.X {type=MovieSceneScriptingDoubleChannel, default=0}
       Rotation.Y {type=MovieSceneScriptingDoubleChannel, default=0}
       Rotation.Z {type=MovieSceneScriptingDoubleChannel, default=2.9878}
       Scale.X {type=MovieSceneScriptingDoubleChannel, default=1.3}
       Scale.Y {type=MovieSceneScriptingDoubleChannel, default=1.3}
       Scale.Z {type=MovieSceneScriptingDoubleChannel, default=1.3}
      ]
     }]
   }
   {
    type=MovieSceneVisibilityTrack
    sections [{
      type=MovieSceneVisibilitySection
      channels [None {type=MovieSceneScriptingBoolChannel, default=true, keys [[-72.9762,true],[-6.0925,false],[474.8612,false],[541,true]]}]
     }]
   }
  ]
 }
 SM_LavaCliff_C20 {
  id=#39
  possessed_class=FortStaticMeshActor
  tracks [{
    type=MovieScene3DTransformTrack
    sections [{
      type=MovieScene3DTransformSection
      channels [
       Location.X {type=MovieSceneScriptingDoubleChannel, default=2878.4963, const [2878.4963,3]}
       Location.Y {type=MovieSceneScriptingDoubleChannel, default=-120663.8984, thinned_from=3, keys [[435.175,-120663.8984],[733.1187,-120663.8984]]}
       Location.Z {type=MovieSceneScriptingDoubleChannel, default=308.7775, const [308.7775,3]}
       Rotation.X {type=MovieSceneScriptingDoubleChannel, default=0}
       Rotation.Y {type=MovieSceneScriptingDoubleChannel, default=0}
       Rotation.Z {type=MovieSceneScriptingDoubleChannel, default=90}
       Scale.X {type=MovieSceneScriptingDoubleChannel, default=0.5}
       Scale.Y {type=MovieSceneScriptingDoubleChannel, default=0.5}
       Scale.Z {type=MovieSceneScriptingDoubleChannel, default=0.5}
      ]
     }]
   }]
 }
 cam_0100 {
  id=#29
  possessed_class=CineCameraActor
  tracks [
   {
    type=MovieScene3DTransformTrack
    sections [{
      type=MovieScene3DTransformSection
      channels [
       Location.X {type=MovieSceneScriptingDoubleChannel, default=3424.4915, thinned_from=126, keys [[0.0013,3424.4915],[125.0013,3185.2549]]}
       Location.Y {type=MovieSceneScriptingDoubleChannel, default=-120093.9766, thinned_from=126, keys [[0.0013,-120093.9766],[125.0013,-120086.0469]]}
       Location.Z {type=MovieSceneScriptingDoubleChannel, default=41.0275, thinned_from=126, keys [[0.0013,41.0275],[125.0013,49.7642]]}
       Rotation.X {type=MovieSceneScriptingDoubleChannel, default=0.2473, thinned_from=126, keys [[0.0013,360.2473],[125.0013,360.2956]]}
       Rotation.Y {type=MovieSceneScriptingDoubleChannel, default=7.1667, thinned_from=126, keys [[0.0013,-352.8333],[125.0013,-352.8557]]}
       Rotation.Z {type=MovieSceneScriptingDoubleChannel, default=167.0128, thinned_from=126, keys [[0.0013,167.0128],[125.0013,167.0434]]}
       Scale.X {type=MovieSceneScriptingDoubleChannel, default=1, const [1,126]}
       Scale.Y {type=MovieSceneScriptingDoubleChannel, default=1, const [1,126]}
       Scale.Z {type=MovieSceneScriptingDoubleChannel, default=1, const [1,126]}
      ]
     }]
   }
   {type=MovieSceneSpawnTrack, sections [{type=MovieSceneSpawnSection, channels [None {type=MovieSceneScriptingBoolChannel, default=true}]}]}
  ]
 }
 BP_Battlepass_LightSwitcher_Maproom {
  id=#40
  possessed_class=BP_Battlepass_LightSwitcher_C
  tracks [{
    type=MovieSceneEventTrack
    sections [{
      type=MovieSceneEventTriggerSection
      channels [None {type=MovieSceneScriptingEventChannel, const ["<Struct 'MovieSceneEvent' (0x8518df541a0) {}>", 1]}]
     }]
   }]
 }
 SpotLight23 {
  id=#16
  possessed_class=SpotLight
  tracks [{
    type=MovieScene3DTransformTrack
    sections [{
      type=MovieScene3DTransformSection
      channels [
       Location.X {type=MovieSceneScriptingDoubleChannel, default=2870.8279, const [2870.8279,3]}
       Location.Y {type=MovieSceneScriptingDoubleChannel, default=-120455.1953, thinned_from=3, keys [[-8.1237,-120455.1953],[732.0013,-120455.1953]]}
       Location.Z {type=MovieSceneScriptingDoubleChannel, default=287.8384, keys [[-8.1237,287.8384],[-7.0888,412.9412],[732.0013,287.8384]]}
       Rotation.X {type=MovieSceneScriptingDoubleChannel, default=265.3893, const [265.3893,2]}
       Rotation.Y {type=MovieSceneScriptingDoubleChannel, default=-62.1987, const [-62.1987,2]}
       Rotation.Z {type=MovieSceneScriptingDoubleChannel, default=-810.0354, const [-810.0354,2]}
       Scale.X {type=MovieSceneScriptingDoubleChannel, default=1, const [1,2]}
       Scale.Y {type=MovieSceneScriptingDoubleChannel, default=1, const [1,2]}
       Scale.Z {type=MovieSceneScriptingDoubleChannel, default=1, const [1,2]}
      ]
     }]
   }]
 }
 CineCameraActor5 {
  id=#21
  possessed_class=CineCameraActor
  tracks [
   {
    type=MovieScene3DTransformTrack
    sections [{
      type=MovieScene3DTransformSection
      channels [
       Location.X {type=MovieSceneScriptingDoubleChannel, default=142.2694}
       Location.Y {type=MovieSceneScriptingDoubleChannel, default=125668.7031}
       Location.Z {type=MovieSceneScriptingDoubleChannel, default=99.4528}
       Rotation.X {type=MovieSceneScriptingDoubleChannel, default=0}
       Rotation.Y {type=MovieSceneScriptingDoubleChannel, default=-0.2}
       Rotation.Z {type=MovieSceneScriptingDoubleChannel, default=-89.2025}
       Scale.X {type=MovieSceneScriptingDoubleChannel, default=1}
       Scale.Y {type=MovieSceneScriptingDoubleChannel, default=1}
       Scale.Z {type=MovieSceneScriptingDoubleChannel, default=1}
      ]
     }]
   }
   {type=MovieSceneSpawnTrack, sections [{type=MovieSceneSpawnSection, channels [None {type=MovieSceneScriptingBoolChannel, default=true}]}]}
  ]
 }
 SM_M_MED_Rider_FaceAcc2 {
  id=#41
  possessed_class=FortStaticMeshActor
  tracks [{
    type=MovieScene3DTransformTrack
    sections [{
      type=MovieScene3DTransformSection
      channels [
       Location.X {type=MovieSceneScriptingDoubleChannel, default=2846.7168, keys [[540,2846.7168],[541,2838.001],[732,2846.7168]]}
       Location.Y {type=MovieSceneScriptingDoubleChannel, default=-120496.3438, const [-120496.3438,3]}
       Location.Z {type=MovieSceneScriptingDoubleChannel, default=153.6741, keys [[540,153.6741],[541,312.0703],[732,153.6741]]}
       Rotation.X {type=MovieSceneScriptingDoubleChannel, default=0, const [0,2]}
       Rotation.Y {type=MovieSceneScriptingDoubleChannel, default=0, const [0,2]}
       Rotation.Z {type=MovieSceneScriptingDoubleChannel, default=-1.5554, const [-1.5554,2]}
       Scale.X {type=MovieSceneScriptingDoubleChannel, default=1.3, const [1.3,2]}
       Scale.Y {type=MovieSceneScriptingDoubleChannel, default=1.3, const [1.3,2]}
       Scale.Z {type=MovieSceneScriptingDoubleChannel, default=1.3, const [1.3,2]}
      ]
     }]
   }]
 }
 BP_Frontend_EventLevel_NavObject_Mannequin2 {
  id=#42
  possessed_class=SoldierMannequin_Male_C
  tracks [
   {
    type=MovieScene3DTransformTrack
    sections [{
      type=MovieScene3DTransformSection
      channels [
       Location.X {type=MovieSceneScriptingDoubleChannel, default=2570.738}
       Location.Y {type=MovieSceneScriptingDoubleChannel, default=-120090.4297}
       Location.Z {type=MovieSceneScriptingDoubleChannel, default=9.8454}
       Rotation.X {type=MovieSceneScriptingDoubleChannel, default=0}
       Rotation.Y {type=MovieSceneScriptingDoubleChannel, default=0}
       Rotation.Z {type=MovieSceneScriptingDoubleChannel, default=-124.3746}
       Scale.X {type=MovieSceneScriptingDoubleChannel, default=1}
       Scale.Y {type=MovieSceneScriptingDoubleChannel, default=1}
       Scale.Z {type=MovieSceneScriptingDoubleChannel, default=1}
      ]
     }]
   }
   {type=MovieSceneSkeletalAnimationTrack, sections []}
   {
    type=MovieSceneVisibilityTrack
    sections [{
      type=MovieSceneVisibilitySection
      channels [None {type=MovieSceneScriptingBoolChannel, default=true, keys [[-8.0312,true],[-7.0163,false],[731,true]]}]
     }]
   }
  ]
 }
]
display_rate {denominator=1, numerator=30}
tick_resolution {denominator=1, numerator=24000}
tracks [
 {
  type=MovieSceneFadeTrack
  sections [{
    type=MovieSceneFadeSection
    channels [None {
      type=MovieSceneScriptingFloatChannel
      default=-0.4035
      keys [[5.15,1.6122],[33.7875,-1.1834],[661.2463,-0.4035],[696.5987,-0.4035],[726,1.6122]]
     }]
   }]
 }
 {
  type=MovieSceneAudioTrack
  display_name=Audio
  sections [{
    type=MovieSceneAudioSection
    end_frame=725
    start_frame=-7
    channels [
     Volume {type=MovieSceneScriptingFloatChannel, default=1}
     Pitch {type=MovieSceneScriptingFloatChannel, default=1}
     AttachActor {type=MovieSceneScriptingActorReferenceChannel, default="<Struct 'MovieSceneObjectBindingID' (0x8515dfe9a80) {}>"}
    ]
   }]
 }
 {
  type=MovieSceneCameraCutTrack
  sections [
   {type=MovieSceneCameraCutSection, end_frame=125, start_frame=-7}
   {type=MovieSceneCameraCutSection, end_frame=242, start_frame=125}
   {type=MovieSceneCameraCutSection, end_frame=476, start_frame=242}
   {type=MovieSceneCameraCutSection, end_frame=541, start_frame=476}
   {type=MovieSceneCameraCutSection, end_frame=731, start_frame=541}
  ]
 }
]
